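# Physics asset for the Candy Summer backpack cosmetic, defining collision bodies and joints for its pack and attached bag.
# The pack body combines a short, wide capsule with a large thin box, while bag_base and bag each use a small rotated capsule.
# Two constraints chain pack to bag_base to bag with angular limits on swing and twist, with looser limits at the outer bag joint.
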PhysicsAsset /BRCosmetics/Accessories/FORT_Backpacks/Backpack_F_MED_Candy_Summer/Meshes/F_MED_Candy_Summer_Pack_Physics
bodies {
 bag [{shapeType=Capsule, length=5, radius=2.5, center=<5.0456 0 0>, rotation=<rot 90 0 0>}]
 bag_base [{shapeType=Capsule, length=1.25, radius=1.5, center=<1.8703 0 0>, rotation=<rot 90 0 0>}]
 pack [
  {shapeType=Capsule, length=0.75, radius=10, center=<-0.1886 -2.5958 -5.9376>}
  {shapeType=Box, extentX=33.5, extentY=3, extentZ=28, center=<-2.9604 5.5556 -10.9092>}
 ]
}
constraints [
 {bone1Name=bag_base, bone2Name=pack, swing1LimitDegrees=15, swing2LimitDegrees=45, twistLimitDegrees=5}
 {bone1Name=bag, bone2Name=bag_base, swing1LimitDegrees=25, swing2LimitDegrees=45, twistLimitDegrees=15}
]
